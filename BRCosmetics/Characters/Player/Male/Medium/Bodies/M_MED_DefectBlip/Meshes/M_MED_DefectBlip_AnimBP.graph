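# Post-copy Anim Blueprint for the DefectBlip player body cosmetic, following the attached parent mesh's pose.
# AnimGraph offsets cape and backpack attach bones and both shoulder plates, then runs a deformation Control Rig.
# Shoulder plate offsets derive each update from clamped upper-arm rotations.
# EventGraph also feeds material-based WPO cloth motion with noise, root velocity and skydiving wind, and pushes a spine-relative virtual pivot to the part's materials.
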AnimBlueprint /BRCosmetics/Characters/Player/Male/Medium/Bodies/M_MED_DefectBlip/Meshes/M_MED_DefectBlip_AnimBP
target_skeleton=&/M_MED_DefectBlip_Skeleton
graphs [
 AnimGraph {nodes [
   n1 = "Transform (Modify) Bone - Bone: attach_cape"(
    ComponentPose=Local To Component(LocalPose=Copy Pose From Mesh() {bUseAttachedParent=true})
    Alpha=1
    Translation="0.000,0.98,-1.12"
    Rotation="0,0,0"
    Scale="1,1,1"
   ) {
    rotationSpace=BCS_BoneSpace
    translationMode=BMM_Additive
    translationSpace=BCS_BoneSpace
    alphaBoolBlend {blendOption=Linear}
    boneToModify {boneName=attach_cape}
    translation {y=0.98, z=-1.12}
   }
   n2 = "Transform (Modify) Bone - Bone: attach_backpack"(ComponentPose=@n1, Alpha=1, Translation="-0.000,1.5,0.000", Rotation="0,0,0", Scale="1,1,1") {
    rotationSpace=BCS_BoneSpace
    translationMode=BMM_Additive
    translationSpace=BCS_BoneSpace
    alphaBoolBlend {blendOption=Linear}
    boneToModify {boneName=attach_backpack}
    translation {y=1.5}
   }
   n3 = "Transform (Modify) Bone - Bone: upperarm_shoulderplate_l"(ComponentPose=@n2, Alpha=1, Translation=Get Left_shoulderplate() -> (Left_shoulderplate:Vector), Rotation="0,0,0", Scale="1,1,1") {
    translationMode=BMM_Additive
    translationSpace=BCS_BoneSpace
    alphaBoolBlend {blendOption=Linear}
    boneToModify {boneName=upperarm_shoulderplate_l}
   }
   n4 = "Transform (Modify) Bone - Bone: upperarm_shoulderplate_r"(ComponentPose=@n3, Alpha=1, Translation=Get Right_shoulderplate() -> (Right_shoulderplate:Vector), Rotation="0,0,0", Scale="1,1,1") {
    translationMode=BMM_Additive
    translationSpace=BCS_BoneSpace
    alphaBoolBlend {blendOption=Linear}
    boneToModify {boneName=upperarm_shoulderplate_r}
   }
   n5 = Control Rig(Source=Component To Local(ComponentPose=@n4)) {
    alphaBoolBlend {blendOption=Linear}
    controlRigAssetReference {blueprintRigClass=&/M_MED_DefectBlip_DeformationRig_ControlRig.M_MED_DefectBlip_DeformationRig_ControlRig_C}
   }
   Output Pose(Result=@n5) {layerGroup=None}
  ]}
 EventGraph {nodes [
   n1 = Event BlueprintUpdateAnimation()
   > n2 = Sequence!(execute) -> (then_0, then_1, then_2:Exec)
   > Set WPO Globals Function(
    Material=Get materialInstanceDynamic() -> (materialInstanceDynamic:Material Instance Dynamic Object Reference)
    Skeletal Mesh=Get SkelMesh()
    "Local Noise Transform  Location"=Get Local Noise Transform Location() -> (Local Noise Transform Location:Vector)
    Local Noise Offset=Get Local Noise Offset()
   )
   > Set WPO Regional Function!(
    execute
    Material:Material Instance Dynamic Object Reference=Get materialInstanceDynamic()
    Overlay Material:Material Instance Dynamic Object Reference
    Skeletal Mesh:Skeletal Mesh Component Object Reference=Get SkelMesh()
    Velocity Transform:Name=root
    Velocity Interp Speed:double=30
    Collision Transform:Name=None
    Collision Transform Axis:EAxis Enum
    Invert Collision Transform Axis:bool=false
    Delta Time X:double=@n1.DeltaTimeX
    Previous Velocity Transform Location:"Vector (by ref)"=Get R1 Velocity Transform Location()
    Previous Velocity Transform Velocity:"Vector (by ref)"=Get R1 Velocity Transform Velocity()
    Region:int=1
   ) -> (then)
   > n3 = Branch(Condition=Get bIsSkyDiving())
   > n4 = Set WPO Pho Wind Intensity(WPO Pho Wind Intensity=1)
   SetScalarParameterValue(execute=[@n5.then, @n4.then], self=Get materialInstanceDynamic(), ParameterName=ClothSim_WindStrength, Value=Get WPO Pho Wind Intensity())
   n5 = Set WPO Pho Wind Intensity(execute=@n3.else, WPO Pho Wind Intensity=0)
   n6 = GetPartSkeletalMeshComponent()
   Is Valid(exec=@n2.then_2, InputObject=@n6)
   > Delay(Duration=0)
   > SetVectorParameterValueOnMaterials!(
    execute
    self:Mesh Component Object Reference=@n6
    ParameterName:Name=DO NOT OVERRIDE-Virtual-Pivot-Ch
    ParameterValue:Vector=TransformLocation!(T:"Transform (by ref)"=GetSocketTransform(self=@n6, InSocketName=spine_05, TransformSpace=RTS_World), Location:Vector="-5, 0,0") -> (ReturnValue:Vector)
   ) -> (then)
   > SetScalarParameterValueOnMaterials!(execute, self:Mesh Component Object Reference=@n6, ParameterName:Name=DO NOT OVERRIDE-Character-Position-BP-Switch, ParameterValue:float=1) -> (then)
   n7 = Get Skeletal Mesh() -> (Skeletal Mesh:Skeletal Mesh Component Object Reference)
   n8 = BreakRotator(InRot=BreakTransform(InTransform=GetSocketTransform(self=@n7, InSocketName=upperarm_l, TransformSpace=RTS_ParentBoneSpace)).Rotation)
   n9 = float * float(A=@n8, B=-0.1) {is=title K2Node_PromotableOperator}
   n10 = float * float(A="Clamp (Float)"(Value=@n8.Pitch, Min=0, Max=90), B=0.1)
   Set Left_shoulderplate(
    execute=@n2.then_1
    Left_shoulderplate:Vector=MakeVector(X=float * float(A=@n10, B=0.2), Y="Clamp (Float)"(Value=@n9, Min=-5, Max=5), Z=@n10)
   ) -> (Output_Get:Vector)
   n11 = Get Skeletal Mesh()
   n12 = BreakRotator(InRot=BreakTransform(InTransform=GetSocketTransform(self=@n11, InSocketName=upperarm_r, TransformSpace=RTS_ParentBoneSpace)).Rotation)
   n13 = float * float(A=@n12, B=0.1)
   n14 = float * float(A="Clamp (Float)"(Value=@n12.Pitch, Min=0, Max=90), B=-0.1)
   > Set Right_shoulderplate(Right_shoulderplate:Vector=MakeVector(X=float * float(A=@n14, B=0.2), Y="Clamp (Float)"(Value=@n13, Min=-5, Max=5), Z=@n14)) -> (Output_Get:Vector)
   Event BlueprintInitializeAnimation()
   > Set Skeletal Mesh(Skeletal Mesh:Skeletal Mesh Component Object Reference=GetOwnerSkeletalMeshComponent()) -> (Output_Get:Skeletal Mesh Component Object Reference)
   Event OnCharacterCustomizationLoaded()
   > n15 = Set SkelMesh(SkelMesh=GetPartSkeletalMeshComponent())
   > n16 = Get Dynamic Material Instance(SkeletalMeshComponent=@n15, ElementIndex=1)
   > Set materialInstanceDynamic(materialInstanceDynamic:Material Instance Dynamic Object Reference=@n16) -> (Output_Get:Material Instance Dynamic Object Reference)
  ]}
]
variables [
 Left_shoulderplate {type=Vector}
 Local Noise Offset {type=Vector}
 Local Noise Transform Location {type=Vector}
 materialInstanceDynamic {type=MaterialInstanceDynamic}
 R1 Velocity Transform Location {type=Vector}
 R1 Velocity Transform Velocity {type=Vector}
 Right_shoulderplate {type=Vector}
 Skeletal Mesh {type=SkeletalMeshComponent}
 skelMesh {type=SkeletalMeshComponent}
 WPO Pho Wind Intensity {type=number}
]
