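# Physics asset for the DefectBlip male medium character's face accessory mesh, giving hair strands, curls and two earrings secondary motion.
# Capsule and small sphere bodies on the dyn_ bones join the head through jointed chains and fans with mostly tight swing and twist limits.
# Most sphere root bodies have fully free angular joints to the head, which carries two capsules and links freely to a neck capsule.
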PhysicsAsset /BRCosmetics/Characters/Player/Male/Medium/Bodies/M_MED_DefectBlip/Meshes/Parts/M_MED_DefectBlip_FaceAcc_Physics
bodies {
 dyn_earing_l_01 [{shapeType=Capsule, length=0.505, radius=0.505, shapeName=dyn_earing_l_01_capsule, center=<0.5795 -0.0244 -0.0287>}]
 dyn_earing_l_02 [{shapeType=Capsule, length=1.6524, radius=0.7527, shapeName=dyn_earing_l_02_capsule, center=<1.8586 -0.1218 0.0432>, rotation=<rot 90 0 0>}]
 dyn_earing_r_01 [{shapeType=Capsule, length=0.505, radius=0.505, shapeName=dyn_earing_r_01_capsule, center=<0.5797 0.0244 -0.0287>}]
 dyn_earing_r_02 [{shapeType=Capsule, length=1.6524, radius=0.7527, shapeName=dyn_earing_r_02_capsule, center=<1.8591 0.1218 0.0432>, rotation=<rot 90 0 0>}]
 dyn_hair_bk_01 [{shapeType=Sphere, radius=0.505, shapeName=dyn_hair_bk_01_sphere}]
 dyn_hair_bk_02 [{
   shapeType=Capsule
   length=0.4914
   radius=0.9462
   shapeName=dyn_hair_bk_02_capsule
   center=<0.493 0.0323 -0.096>
   rotation=<rot -78.9836 -180 -176.3206>
  }]
 dyn_hair_bk_03 [{
   shapeType=Capsule
   length=0.42
   radius=0.8521
   shapeName=dyn_hair_bk_05_capsule
   center=<0.2168 -0.1793 0.0413>
   rotation=<rot -10.5607 48.8464 -81.498>
  }]
 dyn_hair_bk_04 [{
   shapeType=Capsule
   length=0.3595
   radius=0.7725
   shapeName=dyn_hair_bk_04_capsule
   center=<0.0807 -0.0513 0.016>
   rotation=<rot 0.3365 57.5951 -80.5129>
  }]
 dyn_hair_bk_05 [{
   shapeType=Capsule
   length=0.42
   radius=0.8521
   shapeName=dyn_hair_bk_05_capsule
   center=<0.2305 -0.1985 -0.0245>
   rotation=<rot 3.3579 48.9902 -94.6223>
  }]
 dyn_hair_bk_06 [{
   shapeType=Capsule
   length=0.4977
   radius=0.7705
   shapeName=dyn_hair_bk_06_capsule
   center=<0.1906 -0.1853 -0.0313>
   rotation=<rot 6.7986 44.9927 -96.7528>
  }]
 dyn_hair_bk_l_01 [{shapeType=Sphere, radius=0.505, shapeName=dyn_hair_bk_l_01_sphere}]
 dyn_hair_bk_l_03 [{
   shapeType=Capsule
   length=0.4977
   radius=0.7705
   shapeName=dyn_hair_bk_06_capsule
   center=<0.2934 0 0.0045>
   rotation=<rot -89.1253 -0.0002 0.0001>
  }]
 dyn_hair_bk_l_04 [{
   shapeType=Capsule
   length=0.4977
   radius=0.7705
   shapeName=dyn_hair_bk_06_capsule
   center=<0.2399 -0.051 0.0116>
   rotation=<rot -87.237 0.0372 -12.0271>
  }]
 dyn_hair_bk_r_01 [{shapeType=Sphere, radius=0.505, shapeName=dyn_hair_bk_r_01_sphere}]
 dyn_hair_bk_r_02 [{
   shapeType=Capsule
   length=0.4977
   radius=0.7705
   shapeName=dyn_hair_bk_06_capsule
   center=<0.1941 0.1971 -0.013>
   rotation=<rot -6.6512 -44.8639 92.7057>
  }]
 dyn_hair_bk_r_03 [{
   shapeType=Capsule
   length=0.4977
   radius=0.7705
   shapeName=dyn_hair_bk_06_capsule
   center=<0.2754 0.0691 0.0381>
   rotation=<rot -81.5994 -10.5291 24.3846>
  }]
 dyn_hair_curl_01 [{shapeType=Sphere, radius=0.505, shapeName=dyn_hair_curl_01_sphere}]
 dyn_hair_curl_02 [{
   shapeType=Capsule
   length=0.4977
   radius=0.7705
   shapeName=dyn_hair_bk_06_capsule
   center=<-0.0977 -0.3965 0.0379>
   rotation=<rot 1.3001 -13.7244 95.3074>
  }]
 dyn_hair_curl_04 [{shapeType=Capsule, length=0.4977, radius=0.7705, shapeName=dyn_hair_bk_06_capsule, center=<0 -0.4158 -0.1271>, rotation=<rot 0 0 72.9981>}]
 dyn_hair_curl_fr_01 [{shapeType=Sphere, radius=0.505, shapeName=dyn_hair_curl_fr_01_sphere}]
 dyn_hair_curl_fr_02 [{
   shapeType=Capsule
   length=0.9625
   radius=0.7954
   shapeName=dyn_hair_curl_fr_02_capsule
   center=<0.0788 0 -0.3696>
   rotation=<rot -77.9641 -0.0224 0.0102>
  }]
 dyn_hair_curl_fr_03 [{shapeType=Capsule, length=0.505, radius=0.505, shapeName=dyn_hair_curl_fr_03_capsule, center=<0 0.1498 -0.4957>, rotation=<rot 0 0 -13.6987>}]
 dyn_hair_curl_l_01 [{
   shapeType=Capsule
   length=1.4168
   radius=1.1884
   shapeName=dyn_hair_curl_l_01_capsule
   center=<0.5818 0 0.26>
   rotation=<rot -65.9252 0.0002 -0.0001>
  }]
 dyn_hair_curl_l_02 [{
   shapeType=Capsule
   length=1.1716
   radius=1.5997
   shapeName=dyn_hair_curl_l_02_capsule
   center=<0.2428 -0.3283 0.4449>
   rotation=<rot -87.9739 -179.9999 -172.7737>
  }]
 dyn_hair_curl_l_03 [{
   shapeType=Capsule
   length=0.5987
   radius=0.8387
   shapeName=dyn_hair_curl_l_03_capsule
   center=<0.3542 0.129 0.0302>
   rotation=<rot -85.133 0.0003 19.9397>
  }]
 dyn_hair_curl_r_01 [{shapeType=Capsule, length=1.7869, radius=1.1763, shapeName=dyn_hair_curl_r_01_capsule, center=<1.0542 0 -0.1008>, rotation=<rot -90 0 0>}]
 dyn_hair_curl_r_02 [{
   shapeType=Capsule
   length=0.7877
   radius=1.3449
   shapeName=dyn_hair_curl_r_02_capsule
   center=<0.9883 0.3506 -0.2925>
   rotation=<rot -90 180 180>
  }]
 dyn_hair_curl_r_03 [{shapeType=Capsule, length=1.2256, radius=0.934, shapeName=dyn_hair_curl_r_03_capsule, center=<0 -1.321 -0.524>, rotation=<rot 0 0 -120.4336>}]
 dyn_hair_fr_l_01 [{shapeType=Sphere, radius=0.5296, shapeName=dyn_hair_fr_l_01_sphere, center=<3.0733 0.2795 1.8574>}]
 dyn_hair_fr_l_02 [{
   shapeType=Capsule
   length=1.1164
   radius=0.8658
   shapeName=dyn_hair_fr_l_02_capsule
   center=<0.4552 0.0463 -0.0169>
   rotation=<rot -87.869 -179.9805 -174.2215>
  }]
 dyn_hair_fr_l_03 [{
   shapeType=Capsule
   length=0.7468
   radius=0.6729
   shapeName=dyn_hair_fr_l_03_capsule
   center=<0.5634 0.1513 0.0722>
   rotation=<rot -82.701 0.0133 14.9062>
  }]
 dyn_hair_fr_l_04 [{
   shapeType=Capsule
   length=0.6686
   radius=0.821
   shapeName=dyn_hair_fr_l_04_capsule
   center=<0.4607 -0.1836 -0.2366>
   rotation=<rot -62.8235 179.9996 160.4768>
  }]
 dyn_hair_fr_r_01 [{
   shapeType=Capsule
   length=1.187
   radius=1.2546
   shapeName=dyn_hair_fr_r_01_capsule
   center=<0.5029 0 -0.1098>
   rotation=<rot -77.6875 -179.9998 179.9998>
  }]
 dyn_hair_fr_r_02 [{
   shapeType=Capsule
   length=1.5293
   radius=0.9071
   shapeName=dyn_hair_fr_r_02_capsule
   center=<0.3636 0.3262 0.9678>
   rotation=<rot -84.1786 0.0002 10.3114>
  }]
 dyn_hair_fr_r_03 [{
   shapeType=Capsule
   length=1.7292
   radius=0.7849
   shapeName=dyn_hair_fr_r_03_capsule
   center=<0.722 -0.1664 -0.0258>
   rotation=<rot -87.9527 -179.9987 171.5567>
  }]
 dyn_hair_fr_r_04 [{
   shapeType=Capsule
   length=1.0578
   radius=1.0065
   shapeName=dyn_hair_fr_r_04_capsule
   center=<1.1178 -0.6802 0.3059>
   rotation=<rot -74.7307 -0.2254 -22.4813>
  }]
 dyn_hair_l_01 [{shapeType=Sphere, radius=0.505, shapeName=dyn_hair_l_01_sphere}]
 dyn_hair_l_05 [{
   shapeType=Capsule
   length=0.4977
   radius=0.7705
   shapeName=dyn_hair_bk_06_capsule
   center=<0.4997 -0.5142 -0.1776>
   rotation=<rot 3.714 43.2622 76.0583>
  }]
 dyn_hair_r_04 [{
   shapeType=Capsule
   length=0.4977
   radius=0.7705
   shapeName=dyn_hair_bk_06_capsule
   center=<0.2186 -0.4798 0.1086>
   rotation=<rot -0.3108 24.4332 -78.3601>
  }]
 head [
  {shapeType=Capsule, length=2.028, radius=6.7763, center=<6.1156 2.0268 0>, rotation=<rot 0 0 90>}
  {shapeType=Capsule, length=3.54, radius=5.2643, shapeName=head_capsule, center=<0.4657 4.5394 0>, rotation=<rot 0 0 90>}
 ]
 neck_02 [{shapeType=Capsule, length=4.5791, radius=4.8015, shapeName=neck_02_capsule, center=<3.1867 0.6541 0>, rotation=<rot 90 0 0>}]
}
constraints [
 {bone1Name=dyn_hair_bk_02, bone2Name=dyn_hair_bk_01, swing1LimitDegrees=10, swing2LimitDegrees=25, twistLimitDegrees=5}
 {bone1Name=dyn_hair_bk_03, bone2Name=dyn_hair_bk_01, swing1LimitDegrees=25, swing2LimitDegrees=15, twistLimitDegrees=5}
 {bone1Name=dyn_hair_bk_04, bone2Name=dyn_hair_bk_01, swing1LimitDegrees=25, swing2LimitDegrees=10, twistLimitDegrees=5}
 {bone1Name=dyn_hair_bk_05, bone2Name=dyn_hair_bk_01, swing1LimitDegrees=25, swing2LimitDegrees=15, twistLimitDegrees=5}
 {bone1Name=dyn_hair_bk_06, bone2Name=dyn_hair_bk_01, swing1LimitDegrees=25, swing2LimitDegrees=10, twistLimitDegrees=5}
 {
  bone1Name=dyn_hair_bk_01
  bone2Name=head
  swing1LimitDegrees=45
  swing1Motion=Free
  swing2LimitDegrees=45
  swing2Motion=Free
  twistLimitDegrees=45
  twistMotion=Free
 }
 {
  bone1Name=dyn_hair_bk_r_01
  bone2Name=head
  swing1LimitDegrees=45
  swing1Motion=Free
  swing2LimitDegrees=45
  swing2Motion=Free
  twistLimitDegrees=45
  twistMotion=Free
 }
 {
  bone1Name=dyn_hair_bk_l_01
  bone2Name=head
  swing1LimitDegrees=45
  swing1Motion=Free
  swing2LimitDegrees=45
  swing2Motion=Free
  twistLimitDegrees=45
  twistMotion=Free
 }
 {bone1Name=dyn_hair_bk_l_03, bone2Name=dyn_hair_bk_l_01, swing1LimitDegrees=25, swing2LimitDegrees=15, twistLimitDegrees=5}
 {bone1Name=dyn_hair_bk_l_04, bone2Name=dyn_hair_bk_l_01, swing1LimitDegrees=25, swing2LimitDegrees=15, twistLimitDegrees=5}
 {bone1Name=dyn_hair_bk_r_02, bone2Name=dyn_hair_bk_r_01, swing1LimitDegrees=25, swing2LimitDegrees=15, twistLimitDegrees=5}
 {bone1Name=dyn_hair_bk_r_03, bone2Name=dyn_hair_bk_r_01, swing1LimitDegrees=25, swing2LimitDegrees=15, twistLimitDegrees=5}
 {
  bone1Name=dyn_hair_curl_fr_01
  bone2Name=head
  swing1LimitDegrees=45
  swing1Motion=Free
  swing2LimitDegrees=45
  swing2Motion=Free
  twistLimitDegrees=45
  twistMotion=Free
 }
 {bone1Name=dyn_hair_curl_fr_02, bone2Name=dyn_hair_curl_fr_01, swing1LimitDegrees=50, swing2LimitDegrees=35, twistLimitDegrees=10}
 {bone1Name=dyn_hair_curl_fr_03, bone2Name=dyn_hair_curl_fr_02, swing1LimitDegrees=10, swing2LimitDegrees=35, twistLimitDegrees=35}
 {bone1Name=dyn_hair_curl_r_01, bone2Name=dyn_hair_curl_fr_01, swing1LimitDegrees=10, swing2LimitDegrees=10, twistLimitDegrees=10}
 {bone1Name=dyn_hair_curl_r_02, bone2Name=dyn_hair_curl_r_01, swing1LimitDegrees=10, swing2LimitDegrees=10, twistLimitDegrees=10}
 {bone1Name=dyn_hair_curl_r_03, bone2Name=dyn_hair_curl_r_02, swing1LimitDegrees=10, swing2LimitDegrees=10, twistLimitDegrees=10}
 {
  bone1Name=dyn_hair_fr_l_01
  bone2Name=head
  swing1LimitDegrees=45
  swing1Motion=Free
  swing2LimitDegrees=45
  swing2Motion=Free
  twistLimitDegrees=45
  twistMotion=Free
 }
 {bone1Name=dyn_hair_curl_l_01, bone2Name=head, swing1LimitDegrees=15, swing2LimitDegrees=15, twistLimitDegrees=10}
 {bone1Name=dyn_hair_curl_l_02, bone2Name=dyn_hair_curl_l_01, swing1LimitDegrees=15, swing2LimitDegrees=15, twistLimitDegrees=10}
 {bone1Name=dyn_hair_curl_l_03, bone2Name=dyn_hair_curl_l_02, swing1LimitDegrees=15, swing2LimitDegrees=15, twistLimitDegrees=10}
 {bone1Name=dyn_hair_fr_l_02, bone2Name=dyn_hair_fr_l_01, swing1LimitDegrees=15, swing2LimitDegrees=15, twistLimitDegrees=5}
 {bone1Name=dyn_hair_fr_l_03, bone2Name=dyn_hair_fr_l_02, swing1LimitDegrees=15, swing2LimitDegrees=15, twistLimitDegrees=5}
 {bone1Name=dyn_hair_fr_l_04, bone2Name=dyn_hair_fr_l_01, swing1LimitDegrees=25, swing2LimitDegrees=25, twistLimitDegrees=5}
 {bone1Name=dyn_hair_fr_r_01, bone2Name=head, swing1LimitDegrees=10, swing2LimitDegrees=5, twistLimitDegrees=5}
 {bone1Name=dyn_hair_fr_r_02, bone2Name=dyn_hair_fr_r_01, swing1LimitDegrees=15, swing2LimitDegrees=5, twistLimitDegrees=5}
 {bone1Name=dyn_hair_fr_r_03, bone2Name=dyn_hair_fr_r_02, swing1LimitDegrees=25, swing2LimitDegrees=10, twistLimitDegrees=5}
 {
  bone1Name=dyn_hair_curl_01
  bone2Name=head
  swing1LimitDegrees=45
  swing1Motion=Free
  swing2LimitDegrees=45
  swing2Motion=Free
  twistLimitDegrees=45
  twistMotion=Free
 }
 {bone1Name=dyn_hair_curl_02, bone2Name=dyn_hair_curl_01, swing1LimitDegrees=25, swing2LimitDegrees=15, twistLimitDegrees=5}
 {bone1Name=dyn_hair_curl_04, bone2Name=dyn_hair_curl_01, swing1LimitDegrees=25, swing2LimitDegrees=15, twistLimitDegrees=5}
 {
  bone1Name=dyn_hair_l_01
  bone2Name=head
  swing1LimitDegrees=45
  swing1Motion=Free
  swing2LimitDegrees=45
  swing2Motion=Free
  twistLimitDegrees=45
  twistMotion=Free
 }
 {bone1Name=dyn_hair_l_05, bone2Name=dyn_hair_l_01, swing1LimitDegrees=25, swing2LimitDegrees=15, twistLimitDegrees=5}
 {bone1Name=dyn_hair_r_04, bone2Name=head, swing1LimitDegrees=25, swing2LimitDegrees=15, twistLimitDegrees=5}
 {bone1Name=dyn_earing_l_01, bone2Name=head, swing1LimitDegrees=15, swing2LimitDegrees=25, twistLimitDegrees=5}
 {bone1Name=dyn_earing_l_02, bone2Name=dyn_earing_l_01, swing1LimitDegrees=15, swing2LimitDegrees=25, twistLimitDegrees=5}
 {bone1Name=dyn_earing_r_01, bone2Name=head, swing1LimitDegrees=15, swing2LimitDegrees=25, twistLimitDegrees=5}
 {bone1Name=dyn_earing_r_02, bone2Name=dyn_earing_r_01, swing1LimitDegrees=15, swing2LimitDegrees=25, twistLimitDegrees=5}
 {
  bone1Name=head
  bone2Name=neck_02
  swing1LimitDegrees=45
  swing1Motion=Free
  swing2LimitDegrees=45
  swing2Motion=Free
  twistLimitDegrees=45
  twistMotion=Free
 }
 {bone1Name=dyn_hair_fr_r_04, bone2Name=dyn_hair_fr_r_03, swing1LimitDegrees=45, swing2LimitDegrees=10, twistLimitDegrees=5}
]
