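# Animation Blueprint for the DefectBlip face accessory, adding physics secondary motion with a corrective for the back hair bone.
# The AnimGraph copies the attached parent's pose, additively offsets dyn_hair_bk_fix by a variable alpha, then runs RigidBody simulation relative to spine_04 with overridden gravity.
# Each update, the EventGraph derives that alpha from the Face part mesh's parent-space head rotation, taking the stronger of two clamped angle remaps.
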AnimBlueprint /BRCosmetics/Characters/Player/Male/Medium/Bodies/M_MED_DefectBlip/Meshes/Parts/M_MED_DefectBlip_FaceAcc_AnimBP
target_skeleton=&/M_MED_DefectBlip_FaceAcc_Skeleton
graphs [
 AnimGraph {nodes [
   n1 = "Transform (Modify) Bone - Bone: dyn_hair_bk_fix"(
    ComponentPose=Local To Component(LocalPose=Copy Pose From Mesh() {bUseAttachedParent=true})
    Alpha=Get Hair_bk_fix_alpha() -> (Hair_bk_fix_alpha:double)
    Translation="-0.000,-0.000,6"
    Rotation="0,0,0"
    Scale="1,1,1"
   ) {
    translationMode=BMM_Additive
    translationSpace=BCS_BoneSpace
    alphaBoolBlend {blendOption=Linear}
    boneToModify {boneName=dyn_hair_bk_fix}
    translation {z=6}
   }
   n2 = RigidBody(ComponentPose=@n1) {
    bOverrideWorldGravity=true
    overridePhysicsAsset=&/M_MED_DefectBlip_FaceAcc_Physics
    simulationSpace=BaseBoneSpace
    alphaBoolBlend {blendOption=Linear}
    baseBoneRef {boneName=spine_04}
    componentAppliedLinearAccClamp=<3000 3000 3000>
    componentLinearAccScale=<0.3 0.3 0.3>
    componentLinearVelScale=<0.3 0.3 0.3>
    overrideWorldGravity {z=-980}
    simSpaceSettings {maxAngularAcceleration=100, maxAngularVelocity=100, maxLinearAcceleration=100, maxLinearVelocity=100}
   }
   Output Pose(Result=Component To Local(ComponentPose=@n2)) {layerGroup=None}
  ]}
 EventGraph {nodes [
   Event BlueprintUpdateAnimation()
   n1 = Get SkelMesh()
   n2 = BreakRotator(InRot=BreakTransform(InTransform=GetSocketTransform(self=@n1, InSocketName=head, TransformSpace=RTS_ParentBoneSpace)).Rotation)
   n3 = "Absolute (Float)"(A=@n2)
   > Set Hair_bk_fix_alpha(Hair_bk_fix_alpha:double="Max (Float)"(
     A=MapRangeClamped(Value=@n3, InRangeA=0, InRangeB=25)
     B=MapRangeClamped(Value=@n2.Yaw, InRangeA=-42, InRangeB=-15, OutRangeA=1, OutRangeB=0)
    )) -> (Output_Get:double)
   Event BlueprintInitializeAnimation()
   > n4 = Cast To FortPlayerPawnAthena(Object=TryGetPawnOwner())
   > Set SkelMesh(SkelMesh=GetSkeletalMeshForPartType(self=@n4, PartType=Face))
  ]}
]
variables [
 Hair_bk_fix_alpha {type=number}
 SkelMesh {type=SkeletalMeshComponent}
]
